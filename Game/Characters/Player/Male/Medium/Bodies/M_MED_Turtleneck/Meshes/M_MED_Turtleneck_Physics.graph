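# Physics asset for the medium male player turtleneck body, covering only the pelvis and five hip pouches, one left and four right.
# Each pouch is a box body constrained to its own tiny capsule master (MSTR) body with tight, asymmetric swing and minimal twist limits.
# Each master body links to a pelvis capsule with free swing and twist, so angular limits apply only between pouch and master.
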PhysicsAsset /Game/Characters/Player/Male/Medium/Bodies/M_MED_Turtleneck/Meshes/M_MED_Turtleneck_Physics
bodies {
 dyn_l_hipPouch [{shapeType=Box, extentX=7.9074, extentY=4.3638, extentZ=12.4106, center=<1.7239 0.1285 -4.0065>}]
 dyn_l_hipPouch_MSTR [{shapeType=Capsule, length=0.505, radius=0.505}]
 dyn_r_hipPouch_01 [{shapeType=Box, extentX=7.9074, extentY=4.3638, extentZ=12.4106, center=<1.7239 0.1285 -4.0065>}]
 dyn_r_hipPouch_01_MSTR [{shapeType=Capsule, length=0.505, radius=0.505}]
 dyn_r_hipPouch_02 [{shapeType=Box, extentX=7.9074, extentY=4.3638, extentZ=12.4106, center=<1.7239 0.1285 -4.0065>}]
 dyn_r_hipPouch_02_MSTR [{shapeType=Capsule, length=0.505, radius=0.505}]
 dyn_r_hipPouch_03 [{shapeType=Box, extentX=5.9053, extentY=4.3638, extentZ=12.4106, center=<2.1386 0.1284 -3.9512>, rotation=<rot 7.5953 0 0>}]
 dyn_r_hipPouch_03_MSTR [{shapeType=Capsule, length=0.505, radius=0.505}]
 dyn_r_hipPouch_04 [{shapeType=Box, extentX=5.9053, extentY=4.3638, extentZ=12.4106, center=<1.7239 0.1285 -4.0065>}]
 dyn_r_hipPouch_04_MSTR [{shapeType=Capsule, length=0.505, radius=0.505}]
 pelvis [{shapeType=Capsule, length=4.25, radius=11.25, center=<-5.3996 -2.9848 0>, rotation=<rot 0 -90 0>}]
}
constraints [
 {bone1Name=dyn_l_hipPouch, bone2Name=dyn_l_hipPouch_MSTR, swing1LimitDegrees=5, swing2LimitDegrees=30, twistLimitDegrees=1}
 {bone1Name=dyn_r_hipPouch_01, bone2Name=dyn_r_hipPouch_01_MSTR, swing1LimitDegrees=5, swing2LimitDegrees=30, twistLimitDegrees=1}
 {bone1Name=dyn_r_hipPouch_02, bone2Name=dyn_r_hipPouch_02_MSTR, swing1LimitDegrees=5, swing2LimitDegrees=30, twistLimitDegrees=1}
 {bone1Name=dyn_r_hipPouch_03, bone2Name=dyn_r_hipPouch_03_MSTR, swing1LimitDegrees=5, swing2LimitDegrees=30, twistLimitDegrees=1}
 {bone1Name=dyn_r_hipPouch_04, bone2Name=dyn_r_hipPouch_04_MSTR, swing1LimitDegrees=5, swing2LimitDegrees=30, twistLimitDegrees=1}
 {
  bone1Name=dyn_l_hipPouch_MSTR
  bone2Name=pelvis
  swing1LimitDegrees=45
  swing1Motion=Free
  swing2LimitDegrees=45
  swing2Motion=Free
  twistLimitDegrees=45
  twistMotion=Free
 }
 {
  bone1Name=dyn_r_hipPouch_01_MSTR
  bone2Name=pelvis
  swing1LimitDegrees=45
  swing1Motion=Free
  swing2LimitDegrees=45
  swing2Motion=Free
  twistLimitDegrees=45
  twistMotion=Free
 }
 {
  bone1Name=dyn_r_hipPouch_02_MSTR
  bone2Name=pelvis
  swing1LimitDegrees=45
  swing1Motion=Free
  swing2LimitDegrees=45
  swing2Motion=Free
  twistLimitDegrees=45
  twistMotion=Free
 }
 {
  bone1Name=dyn_r_hipPouch_03_MSTR
  bone2Name=pelvis
  swing1LimitDegrees=45
  swing1Motion=Free
  swing2LimitDegrees=45
  swing2Motion=Free
  twistLimitDegrees=45
  twistMotion=Free
 }
 {
  bone1Name=dyn_r_hipPouch_04_MSTR
  bone2Name=pelvis
  swing1LimitDegrees=45
  swing1Motion=Free
  swing2LimitDegrees=45
  swing2Motion=Free
  twistLimitDegrees=45
  twistMotion=Free
 }
]
