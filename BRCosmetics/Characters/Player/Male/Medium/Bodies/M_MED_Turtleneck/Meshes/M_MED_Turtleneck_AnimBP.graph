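# Animation Blueprint for the male medium Turtleneck cosmetic body, adding secondary motion to hip pouches and pickaxe-dependent visibility of material elements.
# It copies the parent mesh pose, switches in left-pouch and knee corrections when jumping, falling or crouching, then applies Anim Dynamics, thigh constraints and crouch-walk pouch offsets.
# The event graph sets CrouchWalking and hides two elements while tagged pickaxes are equipped outside skydiving.
# A RigidBody node sits unconnected.
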AnimBlueprint /BRCosmetics/Characters/Player/Male/Medium/Bodies/M_MED_Turtleneck/Meshes/M_MED_Turtleneck_AnimBP
target_skeleton=&/M_MED_Turtleneck_Skeleton
graphs [
 AnimGraph {nodes [
   RigidBody() {
    bClampLinearTranslationLimitToRefPose=true
    bOverrideWorldGravity=true
    overridePhysicsAsset=/Game/Characters/Player/Male/Medium/Bodies/M_MED_Turtleneck/Meshes/M_MED_Turtleneck_Physics
    simulationSpace=BaseBoneSpace
    alphaBoolBlend {blendOption=Linear}
    baseBoneRef {boneName=root}
    componentLinearAccScale=<0.5 0.5 0.5>
    overrideWorldGravity {z=-980}
   }
   n1 = Component To Local(ComponentPose=Local To Component(LocalPose=Copy Pose From Mesh() {bUseAttachedParent=true}))
   n2 = "Transform (Modify) Bone - Bone: dyn_l_hipPouch_MSTR"(
    ComponentPose=Local To Component(LocalPose=Copy Pose From Mesh() {bUseAttachedParent=true})
    Alpha=1
    Translation="0,-4,0"
    Rotation="0,-25,20"
    Scale="1,1,1"
   ) {
    rotationMode=BMM_Additive
    scaleMode=BMM_Additive
    translationMode=BMM_Additive
    alphaBoolBlend {blendOption=Linear}
    boneToModify {boneName=dyn_l_hipPouch_MSTR}
    rotation {roll=20, yaw=-25}
    translation {y=-4}
   }
   n3 = "Constraint - dyn_knee_l (PP)"!(ComponentPose=@n2, Alpha=1, ConstraintWeights_0=0.5, ConstraintWeights_1:float=0.5) -> (Pose) {
    alphaBoolBlend {blendOption=Linear}
    boneToModify {boneName=dyn_knee_l}
    constraintSetup [{targetBone {boneName=thigh_twist_01_l}}, {targetBone {boneName=calf_l}}]
    constraintWeights [1,1]
   }
   n4 = "Constraint - dyn_knee_r (PP)"!(ComponentPose=@n3, Alpha=1, ConstraintWeights_0=0.6, ConstraintWeights_1:float=0.4) -> (Pose) {
    alphaBoolBlend {blendOption=Linear}
    boneToModify {boneName=dyn_knee_r}
    constraintSetup [{targetBone {boneName=thigh_twist_01_r}}, {targetBone {boneName=calf_r}}]
    constraintWeights [1,1]
   }
   n5 = "Transform (Modify) Bone - Bone: dyn_knee_l"(ComponentPose=@n4, Alpha=1, Translation="0,-3,0", Rotation="0,0,0", Scale="1,1,1.2") {
    rotationMode=BMM_Additive
    rotationSpace=BCS_BoneSpace
    scaleMode=BMM_Additive
    scaleSpace=BCS_BoneSpace
    translationMode=BMM_Additive
    translationSpace=BCS_BoneSpace
    alphaBoolBlend {blendOption=Linear}
    boneToModify {boneName=dyn_knee_l}
    scale {z=1.2}
    translation {y=-3}
   }
   n6 = "Transform (Modify) Bone - Bone: dyn_knee_r"(ComponentPose=@n5, Alpha=1, Translation="0,2,0", Rotation="0,0,0", Scale="1,1,1.2") {
    rotationMode=BMM_Additive
    rotationSpace=BCS_BoneSpace
    scaleMode=BMM_Additive
    scaleSpace=BCS_BoneSpace
    translationMode=BMM_Additive
    translationSpace=BCS_BoneSpace
    alphaBoolBlend {blendOption=Linear}
    boneToModify {boneName=dyn_knee_r}
    scale {z=1.2}
    translation {y=2}
   }
   n7 = Component To Local(ComponentPose=@n6)
   n8 = OR Boolean!(A=Get bIsJumping(), B=Get bIsFalling(), C:bool=Get bIsCrouching()) -> (ReturnValue)
   n9 = Anim Dynamics - Bone: dyn_l_hipPouch(
    ComponentPose=Local To Component(LocalPose=Blend Poses by bool(BlendPose_0=@n7, BlendPose_1=@n1, bActiveValue=@n8) {blendType=Linear})
    Alpha=1
    ExternalForce="0,0,0"
   ) {
    angularDampingOverride=1
    angularSpringConstant=1500
    bAngularSpring=true
    bGravityOverrideInSimSpace=true
    bOverrideAngularDamping=true
    gravityScale=0.35
    simulationSpace=RootRelative
    alphaBoolBlend {blendOption=Linear}
    alphaScaleBiasClamp {scale=0.7}
    boundBone {boneName=dyn_l_hipPouch}
    componentLinearAccScale=<0.4 0.4 0.46>
    componentLinearVelScale=<0.4 0.4 0.46>
    physicsBodyDefinitions [{
      sphereCollisionRadius=0
      boundBone {boneName=dyn_l_hipPouch}
      boxExtents=<10 20 20>
      constraintSetup {angularLimitsMax=<0 20 0>, angularLimitsMin=<0 -20 -40>, angularTarget=<0 0 0>}
      localJointOffset=<5 0 0>
     }]
   }
   n10 = Anim Dynamics - Bone: dyn_r_hipPouch_01(ComponentPose=@n9, Alpha=1, ExternalForce="0,0,0") {
    angularDampingOverride=1
    angularSpringConstant=1500
    bAngularSpring=true
    bGravityOverrideInSimSpace=true
    bOverrideAngularDamping=true
    gravityScale=0.35
    simulationSpace=RootRelative
    alphaBoolBlend {blendOption=Linear}
    alphaScaleBiasClamp {scale=0.7}
    boundBone {boneName=dyn_r_hipPouch_01}
    componentLinearAccScale=<0.4 0.4 0.46>
    componentLinearVelScale=<0.4 0.4 0.46>
    physicsBodyDefinitions [{
      sphereCollisionRadius=0
      boundBone {boneName=dyn_r_hipPouch_01}
      boxExtents=<10 20 20>
      constraintSetup {angularLimitsMax=<0 20 0>, angularLimitsMin=<0 -20 -40>, angularTarget=<0 0 0>}
      localJointOffset=<5 0 0>
     }]
   }
   n11 = Anim Dynamics - Bone: dyn_r_hipPouch_02(ComponentPose=@n10, Alpha=1, ExternalForce="0,0,0") {
    angularDampingOverride=1
    angularSpringConstant=1500
    bAngularSpring=true
    bGravityOverrideInSimSpace=true
    bOverrideAngularDamping=true
    gravityScale=0.35
    simulationSpace=RootRelative
    alphaBoolBlend {blendOption=Linear}
    alphaScaleBiasClamp {scale=0.7}
    boundBone {boneName=dyn_r_hipPouch_02}
    componentLinearAccScale=<0.4 0.4 0.46>
    componentLinearVelScale=<0.4 0.4 0.46>
    physicsBodyDefinitions [{
      sphereCollisionRadius=0
      boundBone {boneName=dyn_r_hipPouch_02}
      boxExtents=<10 20 20>
      constraintSetup {angularLimitsMax=<0 20 0>, angularLimitsMin=<0 -20 -40>, angularTarget=<0 0 0>}
      localJointOffset=<5 0 0>
     }]
   }
   n12 = Anim Dynamics - Bone: dyn_r_hipPouch_03(ComponentPose=@n11, Alpha=1, ExternalForce="0,0,0") {
    angularDampingOverride=1
    angularSpringConstant=1500
    bAngularSpring=true
    bGravityOverrideInSimSpace=true
    bOverrideAngularDamping=true
    gravityScale=0.35
    simulationSpace=RootRelative
    alphaBoolBlend {blendOption=Linear}
    alphaScaleBiasClamp {scale=0.7}
    boundBone {boneName=dyn_r_hipPouch_03}
    componentLinearAccScale=<0.4 0.4 0.46>
    componentLinearVelScale=<0.4 0.4 0.46>
    physicsBodyDefinitions [{
      sphereCollisionRadius=0
      boundBone {boneName=dyn_r_hipPouch_03}
      boxExtents=<10 20 20>
      constraintSetup {angularLimitsMax=<0 20 0>, angularLimitsMin=<0 -20 -40>, angularTarget=<0 0 0>}
      localJointOffset=<5 0 0>
     }]
   }
   n13 = Anim Dynamics - Bone: dyn_r_hipPouch_04(ComponentPose=@n12, Alpha=1, ExternalForce="0,0,0") {
    angularDampingOverride=1
    angularSpringConstant=1500
    bAngularSpring=true
    bGravityOverrideInSimSpace=true
    bOverrideAngularDamping=true
    gravityScale=0.35
    simulationSpace=RootRelative
    alphaBoolBlend {blendOption=Linear}
    alphaScaleBiasClamp {scale=0.7}
    boundBone {boneName=dyn_r_hipPouch_04}
    componentLinearAccScale=<0.4 0.4 0.46>
    componentLinearVelScale=<0.4 0.4 0.46>
    physicsBodyDefinitions [{
      sphereCollisionRadius=0
      boundBone {boneName=dyn_r_hipPouch_04}
      boxExtents=<10 20 20>
      constraintSetup {angularLimitsMax=<0 20 0>, angularLimitsMin=<0 -20 -40>, angularTarget=<0 0 0>}
      localJointOffset=<5 0 0>
     }]
   }
   n14 = "Constraint - dyn_r_hipPouch_01_MSTR (R)"(ComponentPose=@n13, Alpha=1, ConstraintWeights_0=1) {
    alphaBoolBlend {blendOption=Linear}
    boneToModify {boneName=dyn_r_hipPouch_01_MSTR}
    constraintSetup [{transformType=Rotation, targetBone {boneName=thigh_twist_01_r}}]
    constraintWeights [1]
   }
   n15 = "Constraint - dyn_r_hipPouch_02_MSTR (R)"(ComponentPose=@n14, Alpha=0.5, ConstraintWeights_0=1) {
    alphaBoolBlend {blendOption=Linear}
    boneToModify {boneName=dyn_r_hipPouch_02_MSTR}
    constraintSetup [{transformType=Rotation, targetBone {boneName=thigh_twist_01_r}}]
    constraintWeights [1]
   }
   n16 = "Constraint - dyn_r_hipPouch_03_MSTR (R)"(ComponentPose=@n15, Alpha=0.2, ConstraintWeights_0=1) {
    alphaBoolBlend {blendOption=Linear}
    boneToModify {boneName=dyn_r_hipPouch_03_MSTR}
    constraintSetup [{transformType=Rotation, targetBone {boneName=thigh_twist_01_r}}]
    constraintWeights [1]
   }
   n17 = "Transform (Modify) Bone - Bone: dyn_r_hipPouch_02_MSTR"!(
    ComponentPose=@n16
    bAlphaBoolEnabled:bool=Get CrouchWalking() -> (CrouchWalking:bool)
    Translation="0,0,0"
    Rotation="30,0,0"
    Scale="1,1,1"
   ) -> (Pose) {
    alpha=0
    alphaInputType=Bool
    bAlphaBoolEnabled=false
    rotationMode=BMM_Additive
    rotationSpace=BCS_BoneSpace
    scaleMode=BMM_Additive
    scaleSpace=BCS_BoneSpace
    translationMode=BMM_Additive
    translationSpace=BCS_BoneSpace
    alphaBoolBlend {blendInTime=0.1, blendOption=Linear, blendOutTime=0.1}
    boneToModify {boneName=dyn_r_hipPouch_02_MSTR}
    rotation {pitch=30}
   }
   n18 = "Transform (Modify) Bone - Bone: dyn_l_hipPouch"!(ComponentPose=@n17, bAlphaBoolEnabled:bool=True, Translation="0.3,0,0", Rotation="30,0,0", Scale="1,1,1") -> (Pose) {
    alpha=0
    alphaInputType=Bool
    bAlphaBoolEnabled=false
    rotationMode=BMM_Additive
    rotationSpace=BCS_BoneSpace
    scaleMode=BMM_Additive
    scaleSpace=BCS_BoneSpace
    translationMode=BMM_Additive
    translationSpace=BCS_BoneSpace
    alphaBoolBlend {blendInTime=0.1, blendOption=Linear, blendOutTime=0.1}
    boneToModify {boneName=dyn_l_hipPouch}
    rotation {pitch=30}
    translation {x=0.3}
   }
   n19 = "Transform (Modify) Bone - Bone: dyn_l_hipPouch"(ComponentPose=@n18, bAlphaBoolEnabled=Get CrouchWalking(), Translation="0.1,0,0", Rotation="0,-10,0", Scale="1,1,1") {
    alpha=0
    alphaInputType=Bool
    bAlphaBoolEnabled=false
    rotationMode=BMM_Additive
    rotationSpace=BCS_BoneSpace
    scaleMode=BMM_Additive
    scaleSpace=BCS_BoneSpace
    translationMode=BMM_Additive
    translationSpace=BCS_BoneSpace
    alphaBoolBlend {blendInTime=0.1, blendOption=Linear, blendOutTime=0.1}
    boneToModify {boneName=dyn_l_hipPouch}
    rotation {yaw=-10}
    translation {x=0.1}
   }
   Output Pose(Result=Component To Local(ComponentPose=@n19)) {layerGroup=None}
  ]}
 EventGraph {nodes [
   Event BlueprintUpdateAnimation()
   > n1 = Sequence()
   > n2 = Branch(Condition=AND Boolean(A=Get bIsCrouching(), B=Get bIsAccelerating2D()))
   > Set CrouchWalking(CrouchWalking:bool=true) -> (Output_Get:bool)
   Set CrouchWalking(execute=@n2.else, CrouchWalking=false)
   n3 = Get Dynamic Material Instance(execute=@n1.then_1, SkeletalMeshComponent=GetPartSkeletalMeshComponent())
   > n4 = Sequence()
   n5 = PickaxeItemDefinitionHasAnyMetaTag(Tags=Make Array("[0]"=(TagName="Cosmetics.MetaTags.Turtleneck")), bRequiresEquipped=true)
   > n6 = Branch(Condition=@n5)
   > n7 = Branch(Condition=Get bIsSkyDiving())
   SetScalarParameterValue(execute=[@n11.then_0, @n7.then, @n13.else, @n6.else], self=@n3, ParameterName=Hide Element 01, Value=0)
   SetScalarParameterValue(execute=[@n7.else, @n13.then], self=@n3, ParameterName=Hide Element 01, Value=1)
   n8 = PickaxeItemDefinitionHasAnyMetaTag(Tags=Make Array("[0]"=(TagName="Cosmetics.MetaTags.TurtleneckDualWield")), bRequiresEquipped=true)
   n9 = Branch(execute=@n4.then_1, Condition=@n8)
   > n10 = Branch(Condition=Get bIsSkyDiving())
   SetScalarParameterValue(execute=[@n11.then_1, @n9.else, @n10.then, @n14.else], self=@n3, ParameterName=Hide Element 02, Value=0)
   SetScalarParameterValue(execute=[@n10.else, @n14.then], self=@n3, ParameterName=Hide Element 02, Value=1)
   Event OnBeginSkydiving()
   > n11 = Sequence()
   Event OnEndSkydiving()
   > n12 = Sequence()
   > n13 = Branch(Condition=@n5)
   n14 = Branch(execute=@n12.then_1, Condition=@n8)
  ]}
]
variables [CrouchWalking {type=boolean}]
